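# Physics asset for the RoyalAngst glider cosmetic mesh, giving its skeleton simple collision bodies and joint limits.
# Each of the eight bones (frame, two wings, handle base and left/right handle chains) gets a single capsule.
# Constraints form a tree rooted at Frame: both wings and Handles_Base attach to it, and each side's Handle_Base then Handle chains from Handles_Base.
# All joints share identical 45-degree swing and twist limit values.
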PhysicsAsset /BRCosmetics/Gadgets/Assets/VinderTech_GliderChute/Glider_RoyalAngst/Meshes/RoyalAngst_Glider_PhysicsAsset
bodies {
 Frame [{shapeType=Capsule, length=119.6259, radius=67.8862, center=<-0.0493 16.9087 21.0312>, rotation=<rot 0 179.9579 -82.0497>}]
 Handle_Base_l [{shapeType=Capsule, length=49.3695, radius=23.6739, center=<-9.3852 -16.4522 10.5562>, rotation=<rot 13.652 -90 29.0895>}]
 Handle_Base_r [{shapeType=Capsule, length=50.0568, radius=23.5635, center=<7.0386 16.4749 -9.6466>, rotation=<rot 14.1131 -90 30.8484>}]
 Handle_l [{shapeType=Capsule, length=53.837, radius=31.8956, center=<-22.5766 -24.4883 -9.4644>, rotation=<rot 0 -58.5397 -88.212>}]
 Handle_r [{shapeType=Capsule, length=53.8431, radius=31.8798, center=<22.5492 24.5537 9.4695>, rotation=<rot 0 -58.6832 -88.1823>}]
 Handles_Base [{shapeType=Capsule, length=58.9914, radius=50.1916, center=<0.1883 13.4214 10.6793>, rotation=<rot 18.2088 89.539 -90>}]
 Wing_l [{shapeType=Capsule, length=92.1799, radius=27.1716, center=<-67.0267 -1.5361 22.6122>, rotation=<rot 0 179.6193 -89.2009>}]
 Wing_r [{shapeType=Capsule, length=92.1799, radius=27.1716, center=<67.0267 1.5361 -22.6122>, rotation=<rot 0 179.6193 -89.2009>}]
}
constraints [
 {bone1Name=Wing_l, bone2Name=Frame, swing1LimitDegrees=45, swing2LimitDegrees=45, twistLimitDegrees=45}
 {bone1Name=Wing_r, bone2Name=Frame, swing1LimitDegrees=45, swing2LimitDegrees=45, twistLimitDegrees=45}
 {bone1Name=Handles_Base, bone2Name=Frame, swing1LimitDegrees=45, swing2LimitDegrees=45, twistLimitDegrees=45}
 {bone1Name=Handle_Base_r, bone2Name=Handles_Base, swing1LimitDegrees=45, swing2LimitDegrees=45, twistLimitDegrees=45}
 {bone1Name=Handle_r, bone2Name=Handle_Base_r, swing1LimitDegrees=45, swing2LimitDegrees=45, twistLimitDegrees=45}
 {bone1Name=Handle_Base_l, bone2Name=Handles_Base, swing1LimitDegrees=45, swing2LimitDegrees=45, twistLimitDegrees=45}
 {bone1Name=Handle_l, bone2Name=Handle_Base_l, swing1LimitDegrees=45, swing2LimitDegrees=45, twistLimitDegrees=45}
]
